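# Niagara anim-trail effect for the PeacefulPoem pickaxe, drawing short ribbon streaks from bones or sockets.
# Four ribbon emitters form two mirrored pairs offset either way along Z: tinted Claw trails and pale Soft trails, each built with the AnimTrail spawn, initialize and update modules.
# User parameters set loop duration plus Claw width and lifetime, while Soft trails hardcode both.
# Ribbons taper and fade by curve, with time-of-day scaling on opacity and Claw color.
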NiagaraSystem /BRCosmetics/Weapons/FORT_Melee/Pickaxe_PeacefulPoem/FX/NS_Pickaxe_PeacefulPoem_Trail
dependencies {
 usedDataInterfaces [&NiagaraDataInterfaceCurve, &NiagaraDataInterfaceColorCurve]
 usedDynamicInputs []
 usedModules [&SystemState]
}
emitters {
 Claw01 {
  emitterUpdateScript [
   &EmitterState(Life Cycle Mode=::NewEnumerator1, Loop Behavior=::NewEnumerator1, Loop Duration=$User.Loop duration, UseLoopDelay=false)
   &FN_SpawnAnimTrail_v2()
  ]
  particleSpawnScript [
   /Game/Effects/Niagara/Modules/AnimTrail/FN_InitializeAnimTrail/FN_InitializeAnimTrailBoneIndecies(Second Bone or Socket Index=1)
   /Game/Effects/Niagara/Modules/AnimTrail/FN_InitializeAnimTrail/FN_InitializeAnimTrail(
    AnimTrailWidth:NiagaraFloat=$User.AnimTrailWidth
    Color=&FNTimeOfDayColor(
     Night Color=<rgba 0.2078 0.3378 0.6927 1>
     Morning Color=<rgba 0.1675 0.2125 0.3351 1>
     Day Color=<rgba 0.1675 0.2125 0.3351 1>
     Evening Color=<rgba 0.1675 0.2125 0.3351 1>
    )
    Lifetime:NiagaraFloat=$User.Lifetime
    Position Offset:Vector3f=<0 0 3>
   )
  ]
  particleUpdateScript [
   &ScaleRibbonWidth(Ribbon Width Scale=&FloatFromCurve(FloatCurve=DataInterface() {propertyValues {Curve {keys [
         {arriveTangent=-1.2579, interpMode=RCIM_Cubic, time=0, value=1}
         {arriveTangent=-0.9664, interpMode=RCIM_Cubic, leaveTangent=-0.9664, tangentMode=RCTM_User, time=1, value=0}
        ]}}}))
   &ScaleColor(
    Scale Alpha=&FloatFromCurve(
     FloatCurve=DataInterface() {propertyValues {Curve {keys [{arriveTangent=-0.4783, interpMode=RCIM_Cubic, leaveTangent=-0.4783, tangentMode=RCTM_User, time=0, value=0.25}, <at 1 0 RCIM_Cubic>]}}}
     Scale Curve=&FNTimeOfDayScalar(Night Scale=0.5, Morning Scale=0.7, Evening Scale=0.7)
    )
    ScaleA=true
    ScaleRGB=true
   )
   &FN_UpdateAnimTrail_v2()
  ]
  rendererClasses [&NiagaraRibbonRendererProperties]
  renderers [{rendererClass=&NiagaraRibbonRendererProperties}]
 }
 Claw02 {
  emitterUpdateScript [
   &EmitterState(Life Cycle Mode=::NewEnumerator1, Loop Behavior=::NewEnumerator1, Loop Duration=$User.Loop duration, UseLoopDelay=false)
   &FN_SpawnAnimTrail_v2()
  ]
  particleSpawnScript [
   &FN_InitializeAnimTrailBoneIndecies(Second Bone or Socket Index=1)
   &FN_InitializeAnimTrail(
    AnimTrailWidth:NiagaraFloat=$User.AnimTrailWidth
    Color=&FNTimeOfDayColor(
     Night Color=<rgba 0.2078 0.3378 0.6927 1>
     Morning Color=<rgba 0.1675 0.2125 0.3351 1>
     Day Color=<rgba 0.1675 0.2125 0.3351 1>
     Evening Color=<rgba 0.1675 0.2125 0.3351 1>
    )
    Lifetime:NiagaraFloat=$User.Lifetime
    Position Offset:Vector3f=<0 0 -3>
   )
  ]
  particleUpdateScript [
   &ScaleRibbonWidth(Ribbon Width Scale=&FloatFromCurve(FloatCurve=DataInterface() {propertyValues {Curve {keys [
         {arriveTangent=-1.2579, interpMode=RCIM_Cubic, time=0, value=1}
         {arriveTangent=-0.9664, interpMode=RCIM_Cubic, leaveTangent=-0.9664, tangentMode=RCTM_User, time=1, value=0}
        ]}}}))
   &ScaleColor(
    Scale Alpha=&FloatFromCurve(
     FloatCurve=DataInterface() {propertyValues {Curve {keys [{arriveTangent=-0.4783, interpMode=RCIM_Cubic, leaveTangent=-0.4783, tangentMode=RCTM_User, time=0, value=0.25}, <at 1 0 RCIM_Cubic>]}}}
     Scale Curve=&FNTimeOfDayScalar(Night Scale=0.5, Morning Scale=0.7, Evening Scale=0.7)
    )
    ScaleA=true
    ScaleRGB=true
   )
   &FN_UpdateAnimTrail_v2()
  ]
  rendererClasses [&NiagaraRibbonRendererProperties]
  renderers [{rendererClass=&NiagaraRibbonRendererProperties}]
 }
 Soft01 {
  emitterUpdateScript [
   &EmitterState(Life Cycle Mode=::NewEnumerator1, Loop Behavior=::NewEnumerator1, Loop Duration=$User.Loop duration, UseLoopDelay=false)
   &FN_SpawnAnimTrail_v2()
  ]
  particleSpawnScript [
   &FN_InitializeAnimTrailBoneIndecies(First Bone or Socket Index=1)
   &FN_InitializeAnimTrail(AnimTrailWidth=1, Lifetime=0.13, Position Offset:Vector3f=<0 0 3>)
  ]
  particleUpdateScript [
   &ScaleRibbonWidth(Ribbon Width Scale=&FloatFromCurve(FloatCurve=DataInterface() {propertyValues {Curve {keys [
         {arriveTangent=-1.2579, interpMode=RCIM_Cubic, time=0, value=1}
         {arriveTangent=-0.9664, interpMode=RCIM_Cubic, leaveTangent=-0.9664, tangentMode=RCTM_User, time=1, value=0}
        ]}}}))
   &Color(
    Color=&ColorFromCurve(ColorCurve=DataInterface() {propertyValues {
       AlphaCurve {keys [<at 0 0>, {arriveTangent=0.3073, interpMode=RCIM_Cubic, leaveTangent=0.3073, time=0.15, value=0.3}, <at 0.3254 0.1>, <at 1 0 RCIM_Cubic>]}
       BlueCurve {keys [<at 0.2234 1 RCIM_Cubic>,<at 0.3386 1>]}
       GreenCurve {keys [<at 0.2234 0.8413 RCIM_Cubic>,<at 0.3386 0.822>]}
       RedCurve {keys [<at 0.2234 0.8593 RCIM_Cubic>,<at 0.3386 0.85>]}
      }})
    Scale Alpha=&FNTimeOfDayScalar(Night Scale=0.5, Morning Scale=0.75, Day Scale=0.9, Evening Scale=0.75)
   )
   &FN_UpdateAnimTrail_v2()
  ]
  rendererClasses [&NiagaraRibbonRendererProperties]
  renderers [{rendererClass=&NiagaraRibbonRendererProperties}]
 }
 Soft02 {
  emitterUpdateScript [
   &EmitterState(Life Cycle Mode=::NewEnumerator1, Loop Behavior=::NewEnumerator1, Loop Duration=$User.Loop duration, UseLoopDelay=false)
   &FN_SpawnAnimTrail_v2()
  ]
  particleSpawnScript [
   &FN_InitializeAnimTrailBoneIndecies(First Bone or Socket Index=1)
   &FN_InitializeAnimTrail(AnimTrailWidth=1, Lifetime=0.13, Position Offset:Vector3f=<0 0 -3>)
  ]
  particleUpdateScript [
   &ScaleRibbonWidth(Ribbon Width Scale=&FloatFromCurve(FloatCurve=DataInterface() {propertyValues {Curve {keys [
         {arriveTangent=-1.2579, interpMode=RCIM_Cubic, time=0, value=1}
         {arriveTangent=-0.9664, interpMode=RCIM_Cubic, leaveTangent=-0.9664, tangentMode=RCTM_User, time=1, value=0}
        ]}}}))
   &Color(
    Color=&ColorFromCurve(ColorCurve=DataInterface() {propertyValues {
       AlphaCurve {keys [<at 0 0>, {arriveTangent=0.3073, interpMode=RCIM_Cubic, leaveTangent=0.3073, time=0.15, value=0.3}, <at 0.3254 0.1>, <at 1 0 RCIM_Cubic>]}
       BlueCurve {keys [<at 0.2234 1 RCIM_Cubic>,<at 0.3386 1>]}
       GreenCurve {keys [<at 0.2234 0.8413 RCIM_Cubic>,<at 0.3386 0.822>]}
       RedCurve {keys [<at 0.2234 0.8593 RCIM_Cubic>,<at 0.3386 0.85>]}
      }})
    Scale Alpha=&FNTimeOfDayScalar(Night Scale=0.5, Morning Scale=0.75, Day Scale=0.9, Evening Scale=0.75)
   )
   &FN_UpdateAnimTrail_v2()
  ]
  rendererClasses [&NiagaraRibbonRendererProperties]
  renderers [{rendererClass=&NiagaraRibbonRendererProperties}]
 }
}
summary {
 systemName=NS_Pickaxe_PeacefulPoem_Trail
 emitters [
  {emitterName=Claw01, rendererClasses [&NiagaraRibbonRendererProperties]}
  {emitterName=Soft01, rendererClasses [&NiagaraRibbonRendererProperties]}
  {emitterName=Claw02, rendererClasses [&NiagaraRibbonRendererProperties]}
  {emitterName=Soft02, rendererClasses [&NiagaraRibbonRendererProperties]}
 ]
 userVariables [
  User.AnimTrailWidth {type=NiagaraFloat, defaultValue=0.8, description=""}
  User.Lifetime {type=NiagaraFloat, defaultValue=0.12, description=""}
  User.Loop duration {type=NiagaraFloat, defaultValue=0.2, description=""}
  User.Ribbon camera face {type=Vector3f, description="", defaultValue {struct=Vector3f, value=<0 0 0>}}
  User.RibbonIDOffset {type=NiagaraInt32}
  User.SkeletalMesh {
   type=NiagaraDataInterfaceSkeletalMesh
   defaultValue {
    struct=NiagaraExt_VariableValue_DataInterface
    value {dataInterface=&self:NiagaraDataInterfaceSkeletalMesh_0}
   }
  }
 ]
}
